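; Prototype Landmass brush actor that shapes landscape height as one reorderable layer in a BP_Landmass_LayerStack, registering itself on construction.
; DrawBrush renders spline polygons as canvas triangles, or captures spline or static mesh depth, into the stack's depth target.
; Jump flooding turns that into a cacheable distance field before rendering to the stack; one brush type skips this.
; An optional elevation curve is baked into a 256x1 render target.
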(event Selection|EventActorSelectionChanged
  (params (bSelected :type Boolean))
  (Utilities|IsValid (Variables|5%29Debug|GetLS%20BP)
    (:Is%20Valid
      (if bSelected
        (Rendering|Material|SetTextureParameterValue (Variables|/Landmass/Landscape/OldPrototype_BP/BP_Landmass_LayerStack.BP_Landmass_LayerStack_C|GetMeshPreviewMID (Variables|5%29Debug|GetLS%20BP)) "PreviewBrushHeightAlpha" (Variables|5%29Debug|GetCachedHeightAlpha))
        (for _array_element (Variables|/Landmass/Landscape/OldPrototype_BP/BP_Landmass_LayerStack.BP_Landmass_LayerStack_C|GetLayer%20Stack (Variables|5%29Debug|GetLS%20BP))
          (Variables|5%29Debug|Setselected false _array_element))
        (Variables|5%29Debug|Setselected true)
        (elif (Variables|5%29Debug|Getselected)
          (Rendering|Material|SetTextureParameterValue (Variables|/Landmass/Landscape/OldPrototype_BP/BP_Landmass_LayerStack.BP_Landmass_LayerStack_C|GetMeshPreviewMID (Variables|5%29Debug|GetLS%20BP)) "PreviewBrushHeightAlpha" "/Engine/EngineResources/Black.Black")))
      (Variables|5%29Debug|Setselected false))))

(event EventBeginPlay
  (Actor|Tick|SetActorTickEnabled (Variables|5%29Debug|GetTick)))

(event EventTick
  (params (DeltaSeconds :type Float%20%28single-precision%29))
  (2%29Generation|UpdateHeight (Variables|5%29Debug|GetLS%20BP)))

(event Custom|Canvas%20Draw
  (bind (_r _g _b _a) (Utilities|Struct|BreakLinearColor (Variables|Default|GetDynamicPreviewPoint)))
  (bind _ls_bp (Variables|5%29Debug|GetLS%20BP))
  (bind _mesh_depth_rt (Variables|/Landmass/Landscape/OldPrototype_BP/BP_Landmass_LayerStack.BP_Landmass_LayerStack_C|GetMesh%20Depth%20RT _ls_bp))
  (Utilities|Array|Clear (Variables|Default|GetCanvasUVTris))
  (for _index (range (- (Spline|GetNumberOfSplinePoints (Variables|Default|GetSpline)) 2))
    (Utilities|Array|Add (Variables|Default|GetCanvasUVTris) (Utilities|Struct|MakeCanvasUVTri (Math|Conversions|ToVector2D(Vector) (* (+ (/ (Spline|GetLocationatSplinePoint (Variables|Default|GetSpline) 0 "World") (/ (Variables|Default|GetSize) "1.000000,1.000000,1.000000")) 0.5) (Math|Conversions|ToFloat(Integer) (Variables|/Script/Engine.TextureRenderTarget2D|GetSizeX (Variables|/Landmass/Landscape/OldPrototype_BP/BP_Landmass_LayerStack.BP_Landmass_LayerStack_C|GetMesh%20Depth%20RT (Variables|5%29Debug|GetLS%20BP)))))) "(X=0.000000,Y=0.000000)" "(R=1.000000,G=0.000000,B=0.000000,A=1.000000)" (Math|Conversions|ToVector2D(Vector) (* (+ (/ (Spline|GetLocationatSplinePoint (Variables|Default|GetSpline) (+ (+ _index 1) 1) "World") (/ (Variables|Default|GetSize) "1.000000,1.000000,1.000000")) 0.5) (Math|Conversions|ToFloat(Integer) (Variables|/Script/Engine.TextureRenderTarget2D|GetSizeX (Variables|/Landmass/Landscape/OldPrototype_BP/BP_Landmass_LayerStack.BP_Landmass_LayerStack_C|GetMesh%20Depth%20RT (Variables|5%29Debug|GetLS%20BP)))))) "(X=0.000000,Y=0.000000)" "(R=0.000000,G=1.000000,B=0.245000,A=1.000000)" (Math|Conversions|ToVector2D(Vector) (* (+ (/ (Spline|GetLocationatSplinePoint (Variables|Default|GetSpline) (+ _index 1) "World") (/ (Variables|Default|GetSize) "1.000000,1.000000,1.000000")) 0.5) (Math|Conversions|ToFloat(Integer) (Variables|/Script/Engine.TextureRenderTarget2D|GetSizeX (Variables|/Landmass/Landscape/OldPrototype_BP/BP_Landmass_LayerStack.BP_Landmass_LayerStack_C|GetMesh%20Depth%20RT (Variables|5%29Debug|GetLS%20BP)))))) "(X=0.000000,Y=0.000000)" "(R=0.000306,G=0.000000,B=1.000000,A=1.000000)")))
  (if (== _a 1.0)
    (Utilities|FlowControl|Sequence
      (:then_0
        (Utilities|Array|Add (Variables|Default|GetCanvasUVTris) (Utilities|Struct|MakeCanvasUVTri (Math|Conversions|ToVector2D(Vector) (* (+ (/ (Spline|GetLocationatSplinePoint (Variables|Default|GetSpline) 0 "World") (/ (Variables|Default|GetSize) "1.000000,1.000000,1.000000")) 0.5) (Math|Conversions|ToFloat(Integer) (Variables|/Script/Engine.TextureRenderTarget2D|GetSizeX (Variables|/Landmass/Landscape/OldPrototype_BP/BP_Landmass_LayerStack.BP_Landmass_LayerStack_C|GetMesh%20Depth%20RT (Variables|5%29Debug|GetLS%20BP)))))) "(X=0.000000,Y=0.000000)" "(R=1.000000,G=0.000000,B=0.000000,A=1.000000)" (Math|Conversions|ToVector2D(Vector) (* (+ (/ (Math|Conversions|ToVector(LinearColor) (Variables|Default|GetDynamicPreviewPoint)) (/ (Variables|Default|GetSize) "1.000000,1.000000,1.000000")) 0.5) (Math|Conversions|ToFloat(Integer) (Variables|/Script/Engine.TextureRenderTarget2D|GetSizeX (Variables|/Landmass/Landscape/OldPrototype_BP/BP_Landmass_LayerStack.BP_Landmass_LayerStack_C|GetMesh%20Depth%20RT (Variables|5%29Debug|GetLS%20BP)))))) "(X=0.000000,Y=0.000000)" "(R=0.000000,G=1.000000,B=0.245000,A=1.000000)" (Math|Conversions|ToVector2D(Vector) (* (+ (/ (Spline|GetLocationatSplinePoint (Variables|Default|GetSpline) (- (Spline|GetNumberOfSplinePoints (Variables|Default|GetSpline)) 1) "World") (/ (Variables|Default|GetSize) "1.000000,1.000000,1.000000")) 0.5) (Math|Conversions|ToFloat(Integer) (Variables|/Script/Engine.TextureRenderTarget2D|GetSizeX (Variables|/Landmass/Landscape/OldPrototype_BP/BP_Landmass_LayerStack.BP_Landmass_LayerStack_C|GetMesh%20Depth%20RT (Variables|5%29Debug|GetLS%20BP)))))) "(X=0.000000,Y=0.000000)" "(R=0.000306,G=0.000000,B=1.000000,A=1.000000)")))
      (:then_1
        _)))
  (Rendering|ClearRenderTarget2D _mesh_depth_rt)
  (bind (_canvas _size _context) (Rendering|BeginDrawCanvastoRenderTarget _mesh_depth_rt))
  (Utilities|IsValid _canvas
    (:Is%20Valid
      (Utilities|IsValid (Variables|5%29Debug|GetShape%20MID)
        (:Is%20Valid
          _)
        (:Is%20Not%20Valid
          (bind _returnvalue (Rendering|Material|CreateDynamicMaterialInstance "/Landmass/Landscape/OldPrototype_BP/Mats/Debug/Intersections.Intersections" "None" "None" :!class /Script/Engine.KismetMaterialLibrary))
          (bind _output_get (Variables|5%29Debug|SetShape%20MID _returnvalue))))
      (Rendering|Material|SetScalarParameterValue _output_get "Z" (+ (.z (Transformation|GetActorLocation)) (Variables|1%29BrushSettings|Advanced|GetZ%20Offset)) :!class /Script/Engine.MaterialInstanceDynamic)
      (Canvas|DrawMaterialTriangles _canvas (Variables|5%29Debug|GetShape%20MID) (Variables|Default|GetCanvasUVTris))
      (Rendering|EndDrawCanvastoRenderTarget _context))))

(fn 1%29MoveUp
  (bind _self self)
  (bind _ls_bp (Variables|5%29Debug|GetLS%20BP))
  (bind _layer_stack (Variables|/Landmass/Landscape/OldPrototype_BP/BP_Landmass_LayerStack.BP_Landmass_LayerStack_C|GetLayer%20Stack _ls_bp))
  (bind _returnvalue (Utilities|Array|FindItem _layer_stack _self))
  (Utilities|IsValid _ls_bp
    (:Is%20Valid
      (if (>= _returnvalue 0)
        (if (> _returnvalue 0)
          (Utilities|Array|SwapArrayElements (Variables|/Landmass/Landscape/OldPrototype_BP/BP_Landmass_LayerStack.BP_Landmass_LayerStack_C|GetLayer%20Stack (Variables|5%29Debug|GetLS%20BP)) _returnvalue (- _returnvalue 1))
          (2%29Generation|UpdateHeight (Variables|5%29Debug|GetLS%20BP)))
        (else
          (Development|PrintString "Brush was not in layer stack. Try re-running construction script."))))))

(fn 2%29MoveDown
  (bind _self self)
  (bind _ls_bp (Variables|5%29Debug|GetLS%20BP))
  (bind _layer_stack (Variables|/Landmass/Landscape/OldPrototype_BP/BP_Landmass_LayerStack.BP_Landmass_LayerStack_C|GetLayer%20Stack _ls_bp))
  (bind _returnvalue (Utilities|Array|FindItem _layer_stack _self))
  (Utilities|IsValid _ls_bp
    (:Is%20Valid
      (if (>= _returnvalue 0)
        (if (< _returnvalue (Utilities|Array|LastIndex (Variables|/Landmass/Landscape/OldPrototype_BP/BP_Landmass_LayerStack.BP_Landmass_LayerStack_C|GetLayer%20Stack (Variables|5%29Debug|GetLS%20BP))))
          (Utilities|Array|SwapArrayElements (Variables|/Landmass/Landscape/OldPrototype_BP/BP_Landmass_LayerStack.BP_Landmass_LayerStack_C|GetLayer%20Stack (Variables|5%29Debug|GetLS%20BP)) _returnvalue (+ _returnvalue 1))
          (2%29Generation|UpdateHeight (Variables|5%29Debug|GetLS%20BP)))
        (else
          (Development|PrintString "Brush was not in layer stack. Try re-running construction script."))))))

(fn 3%29MovetoTop
  (bind _ls_bp (Variables|5%29Debug|GetLS%20BP))
  (bind _layer_stack (Variables|/Landmass/Landscape/OldPrototype_BP/BP_Landmass_LayerStack.BP_Landmass_LayerStack_C|GetLayer%20Stack _ls_bp))
  (Utilities|IsValid _ls_bp
    (:Is%20Valid
      (for _ (range (+ (Utilities|Array|Length _layer_stack) 1))
        (if (> (Utilities|Array|FindItem _layer_stack self) 0)
          (3%29LayerOrdering|1%29MoveUp)))
      (2%29Generation|UpdateHeight (Variables|5%29Debug|GetLS%20BP)))))

(fn 4%29MovetoBottom
  (bind _ls_bp (Variables|5%29Debug|GetLS%20BP))
  (bind _layer_stack (Variables|/Landmass/Landscape/OldPrototype_BP/BP_Landmass_LayerStack.BP_Landmass_LayerStack_C|GetLayer%20Stack _ls_bp))
  (bind _returnvalue (Utilities|Array|LastIndex _layer_stack))
  (Utilities|IsValid _ls_bp
    (:Is%20Valid
      (for _ (range _returnvalue)
        (if (not (== _returnvalue (Utilities|Array|FindItem _layer_stack self)))
          (3%29LayerOrdering|2%29MoveDown)))
      (2%29Generation|UpdateHeight (Variables|5%29Debug|GetLS%20BP)))))

(fn Cache
  (CallFunction|RenderDistanceField)
  (Variables|5%29Debug|SetDistanceIsCached true))

(fn CenterActoronSplinePoints
  (bind _spline (Variables|Default|GetSpline))
  (bind _returnvalue (Spline|GetNumberOfSplinePoints _spline))
  (bind _returnvalue_2 (- _returnvalue 1))
  (bind _accumavg (LocalVariables|GetAccumAvg))
  (for _index (range _returnvalue)
    (LocalVariables|SetAccumAvg (+ (LocalVariables|GetAccumAvg) (Spline|GetLocationatSplinePoint _spline _index "World"))))
  (LocalVariables|SetAccumAvg (* (/ (LocalVariables|GetAccumAvg) (Math|Conversions|ToFloat(Integer) _returnvalue)) "1.000000,1.000000, 0"))
  (for _index_2 (range _returnvalue)
    (Spline|SetLocationatSplinePoint (Variables|Default|GetSpline) _index_2 (- (Spline|GetLocationatSplinePoint (Variables|Default|GetSpline) _index_2 "World") (LocalVariables|GetAccumAvg)) "World" true))
  (Transformation|AddActorWorldOffset _accumavg)
  (Variables|1%29BrushSettings|Advanced|SetZ%20Offset (- (Variables|1%29BrushSettings|Advanced|GetZ%20Offset) (.z _accumavg)))
  (Variables|Default|SetDynamicPreviewPoint (Math|Color|MakeColor 0.0 0.0 0.0 0.0))
  (Variables|5%29Debug|SetDistanceIsCached false)
  (CallFunction|Constructor)
  (2%29Generation|UpdateHeight (Variables|5%29Debug|GetLS%20BP)))

(fn Constructor
  (bind _spline (Variables|Default|GetSpline))
  (bind _spline_mesh_scale (Variables|1%29BrushSettings|SplineSetup|GetSpline%20Mesh%20Scale))
  (bind _splinemeshesforrender (Variables|Default|GetSplineMeshesForRender))
  (bind _spline_mesh_scale_2 (Variables|1%29BrushSettings|SplineSetup|GetSpline%20Mesh%20Scale))
  (bind _splinemeshes (Variables|Default|GetSplineMeshes))
  (bind _spline_2 (Variables|Default|GetSpline))
  (bind _brush_type (Variables|1%29BrushSettings|GetBrush%20Type))
  (bind _ls_bp (Variables|5%29Debug|GetLS%20BP))
  (Utilities|IsValid (Variables|5%29Debug|GetLS%20BP)
    (:Is%20Valid
      _)
    (:Is%20Not%20Valid
      (bind _outactors (Actor|GetAllActorsOfClass "/Landmass/Landscape/OldPrototype_BP/BP_Landmass_LayerStack.BP_Landmass_LayerStack_C"))
      (if (Utilities|Array|IsNotEmpty _outactors)
        (bind _output_get (Variables|5%29Debug|SetLS%20BP (Utilities|Array|Get(acopy) _outactors))))))
  (Utilities|IsValid _output_get
    (:Is%20Valid
      (Class|BPLandmassLayerStack|Constructor _output_get) ; Ensure constructor is run on LS BP, to make sure we're not dependent on construct order
      (Utilities|FlowControl|Sequence
        (:then_0
          (Variables|Default|SetSize (Math|Conversions|ToVector(Vector2D) (Variables|/Landmass/Landscape/OldPrototype_BP/BP_Landmass_LayerStack.BP_Landmass_LayerStack_C|GetTotalSize (Variables|5%29Debug|GetLS%20BP)) 1.0))
          (Variables|1%29BrushSettings|Material|SetDefault%20Material (select (Utilities|Enum|Equal(Enum) _brush_type "NewEnumerator7") "/Landmass/Landscape/OldPrototype_BP/Mats/Brushes/M_Foundation.M_Foundation" (select (== (Variables|1%29BrushSettings|GetFalloff%20Angle) 0.0) "/Landmass/Landscape/OldPrototype_BP/Mats/Brushes/MeshBrush_Width.MeshBrush_Width" "/Landmass/Landscape/OldPrototype_BP/Mats/Brushes/MeshBrush_Angle.MeshBrush_Angle")))
          (switch Utilities|FlowControl|Switch|SwitchonBrushTypes _brush_type
            (:NewEnumerator1
              ;[1 1 1] Mesh Specific Setups
              (Utilities|IsValid (Variables|1%29BrushSettings|SplineSetup|GetSpline%20Mesh)
                (:Is%20Valid
                  (if (> (Spline|GetNumberOfSplinePoints (Variables|Default|GetSpline)) 2)
                    ;; Spawn Spline Meshes
                    (Utilities|IsValid (Variables|1%29BrushSettings|SplineSetup|GetSpline%20Mesh)
                      (:Is%20Valid
                        (for _array_element _splinemeshes
                          (Components|DestroyComponent _array_element))
                        (Utilities|Array|Clear _splinemeshes)
                        (for _array_element_2 _splinemeshesforrender
                          (Components|DestroyComponent _array_element_2))
                        (Utilities|Array|Clear _splinemeshesforrender)
                        (Utilities|FlowControl|Sequence
                          (:then_0
                            (for _index (range (+ (- (Spline|GetNumberOfSplinePoints (Variables|Default|GetSpline)) 2) 1))
                              (bind _returnvalue (AddComponent|NODE_AddSplineMeshComponent-0 :RelativeTransform (Math|Transform|MakeTransform "0, 0, 0")))
                              (Components|StaticMesh|SetStaticMesh _returnvalue (Variables|1%29BrushSettings|SplineSetup|GetSpline%20Mesh))
                              (Rendering|Material|SetMaterial _returnvalue 0 (Variables|1%29BrushSettings|SplineSetup|GetSpline%20Material))
                              (SplineMesh|SetStartScale _returnvalue (Math|Conversions|ToVector2D(Vector) (* _spline_mesh_scale (Math|Vector|MakeVector (.y (Spline|GetScaleatSplinePoint _spline_2 _index)) (.z (Spline|GetScaleatSplinePoint _spline_2 _index))))) false :!class /Script/Engine.SplineMeshComponent)
                              (SplineMesh|SetEndScale _returnvalue (Math|Conversions|ToVector2D(Vector) (* _spline_mesh_scale (Math|Vector|MakeVector (.y (Spline|GetScaleatSplinePoint _spline_2 (+ _index 1))) (.z (Spline|GetScaleatSplinePoint _spline_2 (+ _index 1)))))) false :!class /Script/Engine.SplineMeshComponent)
                              (SplineMesh|SetStartandEnd _returnvalue (Spline|GetLocationandTangentatSplinePoint _spline_2 _index "Local") (.Tangent (Spline|GetLocationandTangentatSplinePoint _spline_2 _index "Local")) (Spline|GetLocationandTangentatSplinePoint _spline_2 (+ _index 1) "Local") (.Tangent (Spline|GetLocationandTangentatSplinePoint _spline_2 (+ _index 1) "Local")) true :!class /Script/Engine.SplineMeshComponent)
                              (Utilities|Array|Add (Variables|Default|GetSplineMeshes) _returnvalue)
                              (Rendering|SetVisibility _returnvalue (Variables|1%29BrushSettings|SplineSetup|GetShow%20Splines) false)
                              (Utilities|IsValid (Variables|1%29BrushSettings|SplineSetup|GetFlatten%20Mesh)
                                (:Is%20Not%20Valid
                                  (Utilities|Array|Add (Variables|Default|GetSplineMeshesForRender) _returnvalue)))))
                          (:then_1
                            (Utilities|IsValid (Variables|1%29BrushSettings|SplineSetup|GetFlatten%20Mesh)
                              (:Is%20Valid
                                (for _index_2 (range (+ (- (Spline|GetNumberOfSplinePoints (Variables|Default|GetSpline)) 2) 1))
                                  (bind _returnvalue_2 (AddComponent|NODE_AddSplineMeshComponent-1 :RelativeTransform (Math|Transform|MakeTransform "0, 0, 0")))
                                  (Components|StaticMesh|SetStaticMesh _returnvalue_2 (Variables|1%29BrushSettings|SplineSetup|GetFlatten%20Mesh))
                                  (SplineMesh|SetStartScale _returnvalue_2 (Math|Conversions|ToVector2D(Vector) (* _spline_mesh_scale_2 (Math|Vector|MakeVector (.y (Spline|GetScaleatSplinePoint _spline _index_2)) (.z (Spline|GetScaleatSplinePoint _spline _index_2))))) false :!class /Script/Engine.SplineMeshComponent)
                                  (SplineMesh|SetEndScale _returnvalue_2 (Math|Conversions|ToVector2D(Vector) (* _spline_mesh_scale_2 (Math|Vector|MakeVector (.y (Spline|GetScaleatSplinePoint _spline (+ _index_2 1))) (.z (Spline|GetScaleatSplinePoint _spline (+ _index_2 1)))))) false :!class /Script/Engine.SplineMeshComponent)
                                  (SplineMesh|SetStartandEnd _returnvalue_2 (Spline|GetLocationandTangentatSplinePoint _spline _index_2 "Local") (.Tangent (Spline|GetLocationandTangentatSplinePoint _spline _index_2 "Local")) (Spline|GetLocationandTangentatSplinePoint _spline (+ _index_2 1) "Local") (.Tangent (Spline|GetLocationandTangentatSplinePoint _spline (+ _index_2 1) "Local")) true :!class /Script/Engine.SplineMeshComponent)
                                  (Utilities|Array|Add (Variables|Default|GetSplineMeshesForRender) _returnvalue_2)
                                  (Rendering|SetVisibility _returnvalue_2 (Variables|1%29BrushSettings|SplineSetup|GetShow%20Flatten%20Mesh) false))))))))))))
            (:NewEnumerator5
              (Utilities|IsValid (Variables|1%29BrushSettings|SingleMeshSetup|GetSingle%20Mesh)
                (:Is%20Valid
                  (Components|StaticMesh|SetStaticMesh (Variables|Default|GetStaticMesh1) (Variables|1%29BrushSettings|SingleMeshSetup|GetSingle%20Mesh)))))))
        (:then_1
          (Utilities|FlowControl|Sequence
            (:then_0
              (if (== (Variables|LayerOrdering|GetStack%20Index) -1)
                (Utilities|IsValid _ls_bp
                  (:Is%20Valid
                    (if (not (>= (Utilities|Array|FindItem (Variables|/Landmass/Landscape/OldPrototype_BP/BP_Landmass_LayerStack.BP_Landmass_LayerStack_C|GetLayer%20Stack _ls_bp) self) 0))
                      (Utilities|Array|Insert (Variables|/Landmass/Landscape/OldPrototype_BP/BP_Landmass_LayerStack.BP_Landmass_LayerStack_C|GetLayer%20Stack (Variables|5%29Debug|GetLS%20BP)) self))))))
            (:then_1
              (if (Variables|2%29Effects|Curves|GetUse%20Elevation%20Curve)
                (Utilities|IsValid (Variables|2%29Effects|Curves|GetElevation%20Remap%20Curve)
                  (:Is%20Valid
                    (Utilities|IsValid (Variables|2%29Effects|Curves|GetCurve%20RT)
                      (:Is%20Valid
                        (if (Variables|5%29Debug|GetRenderedCurveTexture)
                          _
                          (else
                            _)))
                      (:Is%20Not%20Valid
                        (bind _returnvalue_3 (Rendering|CreateRenderTarget2D 256 1 "RTF_R16f"))
                        (Variables|2%29Effects|Curves|SetCurve%20RT _returnvalue_3)))
                    (2%29Effects|UpdateCurves)))))))))))

(fn DebugDetectEdges
  (bind _ls_bp (Variables|5%29Debug|GetLS%20BP))
  (bind _jumpflood_component2d (Variables|/Landmass/Landscape/OldPrototype_BP/BP_Landmass_LayerStack.BP_Landmass_LayerStack_C|GetJumpFlood_Component2D _ls_bp))
  (bind _jump_flood_target_a_vel (Variables|/Landmass/Landscape/OldPrototype_BP/BP_Landmass_LayerStack.BP_Landmass_LayerStack_C|GetJump%20Flood%20Target%20A_Vel _ls_bp))
  (bind _ls_bp_2 (Variables|5%29Debug|GetLS%20BP))
  (Class|JumpFloodComponent2D|AssignRenderTargets _jumpflood_component2d _jump_flood_target_a_vel (Variables|/Landmass/Landscape/OldPrototype_BP/BP_Landmass_LayerStack.BP_Landmass_LayerStack_C|GetJump%20Flood%20Target%20B_Vel _ls_bp) :!class /Landmass/DistanceFields/Blueprints/JumpFlood_Component2D.JumpFlood_Component2D_C)
  (Rendering|ClearRenderTarget2D (Variables|/Landmass/Landscape/OldPrototype_BP/BP_Landmass_LayerStack.BP_Landmass_LayerStack_C|GetNoise%20Seed%20%20RT _ls_bp))
  (Rendering|ClearRenderTarget2D (Variables|/Landmass/Landscape/OldPrototype_BP/BP_Landmass_LayerStack.BP_Landmass_LayerStack_C|GetJump%20Flood%20Target%20A_Vel _ls_bp_2))
  (Rendering|ClearRenderTarget2D (Variables|/Landmass/Landscape/OldPrototype_BP/BP_Landmass_LayerStack.BP_Landmass_LayerStack_C|GetJump%20Flood%20Target%20B_Vel _ls_bp_2))
  (Class|JumpFloodComponent2D|FindEdgesDebug _jumpflood_component2d (Variables|/Landmass/Landscape/OldPrototype_BP/BP_Landmass_LayerStack.BP_Landmass_LayerStack_C|GetMesh%20Depth%20RT _ls_bp) (select (Variables|1%29BrushSettings|GetUseMeshDepth) 0.0 0.0) (Utilities|Struct|MakeLinearColor (Variables|2%29Effects|CurlNoise|GetCurl%201%20Tiling) (Variables|2%29Effects|CurlNoise|GetCurl%201%20Strength) (Variables|2%29Effects|CurlNoise|GetCurl%202%20Tiling) (Variables|2%29Effects|CurlNoise|GetCurl%202%20Strength)) _jump_flood_target_a_vel 0.0 :!class /Landmass/DistanceFields/Blueprints/JumpFlood_Component2D.JumpFlood_Component2D_C))

(fn DrawBrush
  (bind (_r _g _b _a) (Utilities|Struct|BreakLinearColor (Variables|Default|GetDynamicPreviewPoint)))
  (bind _splinemeshesforrender (Variables|Default|GetSplineMeshesForRender))
  (bind _staticmesh1 (Variables|Default|GetStaticMesh1))
  (Utilities|IsValid (Variables|5%29Debug|GetLS%20BP)
    (:Is%20Valid
      ;[1 1 1] Check if anything Cacheable
      (if (not (Utilities|Enum|Equal(Enum) (Variables|1%29BrushSettings|GetBrush%20Type) "NewEnumerator7"))
        (if (Variables|5%29Debug|GetDistanceIsCached)
          (CallFunction|RenderBrushtoStack)
          (else
            (switch Utilities|FlowControl|Switch|SwitchonBrushTypes (Variables|1%29BrushSettings|GetBrush%20Type)
              (:NewEnumerator0
                ;[1 1 1] User Drawn Shape
                (if (> (+ (Spline|GetNumberOfSplinePoints (Variables|Default|GetSpline)) (select (== _a 1.0) 1 0)) 2)
                  (CallFunction|CanvasDraw)
                  (Class|JumpFloodComponent2D|JumpFlood (Variables|/Landmass/Landscape/OldPrototype_BP/BP_Landmass_LayerStack.BP_Landmass_LayerStack_C|GetJumpFlood_Component2D (Variables|5%29Debug|GetLS%20BP)) (Variables|/Landmass/Landscape/OldPrototype_BP/BP_Landmass_LayerStack.BP_Landmass_LayerStack_C|GetMesh%20Depth%20RT (Variables|5%29Debug|GetLS%20BP)) 0.0 (Utilities|Struct|MakeLinearColor (Variables|2%29Effects|CurlNoise|GetCurl%201%20Tiling) (Variables|2%29Effects|CurlNoise|GetCurl%201%20Strength) (Variables|2%29Effects|CurlNoise|GetCurl%202%20Tiling) (Variables|2%29Effects|CurlNoise|GetCurl%202%20Strength)) false (+ (.z (Transformation|GetActorLocation)) (Variables|1%29BrushSettings|Advanced|GetZ%20Offset)) :!class /Landmass/DistanceFields/Blueprints/JumpFlood_Component2D.JumpFlood_Component2D_C)))
              (:NewEnumerator1
                ;[1 1 1] Spline Mesh
                (if (> (Spline|GetNumberOfSplinePoints (Variables|Default|GetSpline)) 2)
                  (for _array_element _splinemeshesforrender
                    (Rendering|SetVisibility _array_element true false))
                  (Class|BPLandmassLayerStack|CaptureSplineDepth (Variables|5%29Debug|GetLS%20BP) _splinemeshesforrender)
                  (for _array_element_2 _splinemeshesforrender
                    (Rendering|SetVisibility _array_element_2 (select (Utilities|IsValid (Variables|1%29BrushSettings|SplineSetup|GetFlatten%20Mesh) :!class /Script/Engine.KismetSystemLibrary) (Variables|1%29BrushSettings|SplineSetup|GetShow%20Flatten%20Mesh) (Variables|1%29BrushSettings|SplineSetup|GetShow%20Splines)) false))
                  (Class|JumpFloodComponent2D|JumpFlood (Variables|/Landmass/Landscape/OldPrototype_BP/BP_Landmass_LayerStack.BP_Landmass_LayerStack_C|GetJumpFlood_Component2D (Variables|5%29Debug|GetLS%20BP)) (Variables|/Landmass/Landscape/OldPrototype_BP/BP_Landmass_LayerStack.BP_Landmass_LayerStack_C|GetMesh%20Depth%20RT (Variables|5%29Debug|GetLS%20BP)) 5e+04 (Utilities|Struct|MakeLinearColor (Variables|2%29Effects|CurlNoise|GetCurl%201%20Tiling) (Variables|2%29Effects|CurlNoise|GetCurl%201%20Strength) (Variables|2%29Effects|CurlNoise|GetCurl%202%20Tiling) (Variables|2%29Effects|CurlNoise|GetCurl%202%20Strength)) true (+ (* (.z (Transformation|GetActorLocation (Variables|5%29Debug|GetLS%20BP))) -1.0) (Variables|1%29BrushSettings|Advanced|GetZ%20Offset)) :!class /Landmass/DistanceFields/Blueprints/JumpFlood_Component2D.JumpFlood_Component2D_C :!label _label)))
              (:NewEnumerator5
                ;[1 1 1] Regular Static Mesh
                (Development|SetHiddeninGame _staticmesh1 false false)
                (Class|BPLandmassLayerStack|CaptureMeshDepth (Variables|5%29Debug|GetLS%20BP) _staticmesh1)
                (Development|SetHiddeninGame _staticmesh1 true false)
                (goto _label)))
            (CallFunction|RenderDistanceField))))
      (CallFunction|RenderBrushtoStack))))

(fn JumpFloodSingleStep
  (Class|JumpFloodComponent2D|SingleJumpStep (Variables|/Landmass/Landscape/OldPrototype_BP/BP_Landmass_LayerStack.BP_Landmass_LayerStack_C|GetJumpFlood_Component2D (Variables|5%29Debug|GetLS%20BP)) :!class /Landmass/DistanceFields/Blueprints/JumpFlood_Component2D.JumpFlood_Component2D_C))

(fn Refresh
  (bind _self self)
  (bind _ls_bp (Variables|5%29Debug|GetLS%20BP))
  (bind _layer_stack (Variables|/Landmass/Landscape/OldPrototype_BP/BP_Landmass_LayerStack.BP_Landmass_LayerStack_C|GetLayer%20Stack _ls_bp))
  (bind _returnvalue (Utilities|Array|FindItem _layer_stack _self))
  (Utilities|IsValid _ls_bp
    (:Is%20Valid
      (if (>= _returnvalue 0)
        (if (< _returnvalue (Utilities|Array|LastIndex (Variables|/Landmass/Landscape/OldPrototype_BP/BP_Landmass_LayerStack.BP_Landmass_LayerStack_C|GetLayer%20Stack (Variables|5%29Debug|GetLS%20BP))))
          (Utilities|Array|SwapArrayElements (Variables|/Landmass/Landscape/OldPrototype_BP/BP_Landmass_LayerStack.BP_Landmass_LayerStack_C|GetLayer%20Stack (Variables|5%29Debug|GetLS%20BP)) _returnvalue (Utilities|Array|LastIndex (Variables|/Landmass/Landscape/OldPrototype_BP/BP_Landmass_LayerStack.BP_Landmass_LayerStack_C|GetLayer%20Stack (Variables|5%29Debug|GetLS%20BP))))
          (2%29Generation|UpdateHeight (Variables|5%29Debug|GetLS%20BP)))
        (else
          (Development|PrintString "Brush was not in layer stack. Try re-running construction script."))))))

(fn Update
  (1%29BrushSettings|UpdateLandscape))

(fn UpdateCurves
  (Utilities|IsValid (Variables|2%29Effects|Curves|GetElevation%20Remap%20Curve)
    (:Is%20Valid
      (Rendering|ClearRenderTarget2D (Variables|2%29Effects|Curves|GetCurve%20RT))
      (bind _canvas (Rendering|BeginDrawCanvastoRenderTarget (Variables|2%29Effects|Curves|GetCurve%20RT)))
      (Utilities|IsValid _canvas
        (:Is%20Valid
          (for _index (range 256)
            (Canvas|DrawBox _canvas (Math|Vector2D|MakeVector2D (Math|Conversions|ToFloat(Integer) _index)) "(X=1.000000,Y=2.000000)" 1.0 (Math|Conversions|ToLinearColor(Float) (Math|Curves|GetFloatValue (Variables|2%29Effects|Curves|GetElevation%20Remap%20Curve) (/ (Math|Conversions|ToFloat(Integer) _index) 255.0)))))
          (Variables|5%29Debug|SetRenderedCurveTexture true))))))

(fn UpdateLandscape
  (2%29Generation|UpdateHeight (Variables|5%29Debug|GetLS%20BP)))

(fn ConstructionScript
  (CallFunction|Constructor)
  (Variables|5%29Debug|SetDistanceIsCached false)
  (2%29Generation|UpdateBrushesandNormals (Variables|5%29Debug|GetLS%20BP)))
